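; Rune area-effect actor owned by the Figment Cube, applying a gameplay effect to player pawns inside its capsule and removing it on exit.
; The server enables ambient rune particles immediately or, for the cube's current rune, when it leaves position; clients disable collision and play looping audio.
; Replication callbacks set rune transform and material, ambient particles, and printing glow on the cube's down face.
; Construction scales visuals from capsule radius; replay scrubbing restores particles.
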(event Custom|TrySpawnGlowEffect)

(event Custom|CubeNoLongerOnTop
  (bind _ascube_figment (Utilities|Casting|CastToCUBE_Figment (Actor|GetOwner)))
  (Variables|Default|SetCubeFigment _ascube_figment)
  (Development|PrintString "Hello")
  (Variables|Default|SetEnableRuneAmbientEffect true))

(event Custom|SpawnVolumeEffect
  (Variables|Default|SetEnableRuneAmbientEffect true))

(event EventBeginPlay
  (bind _returnvalue (BlueprintContext|GetFortReplayContext))
  (Utilities|FlowControl|Sequence
    (:then_0
      (Networking|SwitchHasAuthority
        (:Authority
          (bind _ascube_figment (Utilities|Casting|CastToCUBE_Figment (Actor|GetOwner)))
          (bind _output_get (Variables|Default|SetCubeFigment _ascube_figment))
          (bind (_currentstep_t_16_ea4278f841ee4e73c6861e88cb99517b _currentstep_2parttransitiontime_12_209590664625dbbf3f6200a38992da89 _currentstep_downface_19_20869c6c4df1517848398693617bdfc2 _currentstep_runenumber_21_3c4aa7924813a36f10d2fa820589c6a3) (Utilities|Struct|BreakCubeStruct (Variables|/Figment_S05_Cube/Athena/Prototype/Blueprints/Cube/CUBE_Figment.CUBE_Figment_C|GetCurrentStep_T_16_EA4278F841EE4E73C6861E88CB99517B _output_get)))
          (if (Math|Byte|Equal(Byte) _currentstep_runenumber_21_3c4aa7924813a36f10d2fa820589c6a3 (Math|Byte|Byte+Byte (Variables|Default|GetRuneIndex) 1))
            (Default|BindEventtoCubeLeftPosition (Variables|Default|GetCubeFigment) (delegate Custom|SpawnVolumeEffect))
            (else
              (Variables|Default|SetEnableRuneAmbientEffect true))))
        (:Remote
          (Collision|SetCollisionEnabled (Variables|Default|GetEffect%20Area) "NoCollision")
          (bind _returnvalue_2 (Audio|SpawnSoundAttached "/Game/Athena/Sounds/Events/Cube/Cube_AreaEffect_LeftBehind_Loop_Cue.Cube_AreaEffect_LeftBehind_Loop_Cue" (Variables|Default|GetEffect%20Area)))
          (Audio|Components|Audio|FadeIn _returnvalue_2 1e+01 1.0 0.0 "Linear" :!class /Script/Engine.AudioComponent))))
    (:then_1
      (Utilities|IsValid _returnvalue
        (:Is%20Valid
          (ReplayContext|BindEventtoOnReplayScrubComplete _returnvalue (delegate Custom|replayscrubbed)))))))

(event Custom|replayscrubbed
  (Components|Activation|Activate (Variables|Default|GetP_Cube_AoE_Volume) true)
  (Components|Activation|SetActive (Variables|Default|GetP_Cube_Rune_Volume) (Variables|Default|GetEnableRuneAmbientEffect) true))

(event Custom|DisableGlowEffect
  (Variables|Default|SetPrinting false))

(event OnComponentBeginOverlap(EffectArea)
  (params (OverlappedComponent :type Primitive%20Component%20Object%20Reference) (OtherActor :type Actor%20Object%20Reference) (OtherComp :type Primitive%20Component%20Object%20Reference) (OtherBodyIndex :type Integer) (bFromSweep :type Boolean) (SweepResult :type Hit%20Result%20Structure%20%28by%20ref%29))
  (bind _asfort_player_pawn (Utilities|Casting|CastToFortPlayerPawn OtherActor))
  (Environment|ApplyGlobalEnvironmentGameplayEffecttoActor _asfort_player_pawn "/Figment_S05_Cube/Athena/Effects/GE_Cube_Figment_AreaEffect.GE_Cube_Figment_AreaEffect_C"))

(event Collision|EventActorEndOverlap
  (params (OtherActor :type Actor%20Object%20Reference))
  (bind _asfort_player_pawn (Utilities|Casting|CastToFortPlayerPawn OtherActor))
  (GameplayEffects|RemoveActiveGameplayEffectbySourceEffect (Ability|GetAbilitySystemComponent _asfort_player_pawn) "/Figment_S05_Cube/Athena/Effects/GE_Cube_Figment_AreaEffect.GE_Cube_Figment_AreaEffect_C" 0 -1))

(fn OnRep_EnableRuneAmbientEffect
  (Networking|SwitchHasAuthority
    (:Remote
      (Components|Activation|SetActive (Variables|Default|GetP_Cube_Rune_Volume) (Variables|Default|GetEnableRuneAmbientEffect) true))))

(fn OnRep_Printing
  (bind _p_cube_printing (Variables|Default|GetP_Cube_Printing))
  (if (Variables|Default|GetPrinting)
    (bind _ascube_figment (Utilities|Casting|CastToCUBE_Figment (Actor|GetOwner)))
    (bind _output_get (Variables|Default|SetCubeFigment _ascube_figment))
    (bind (_currentstep_t_16_ea4278f841ee4e73c6861e88cb99517b _currentstep_2parttransitiontime_12_209590664625dbbf3f6200a38992da89 _currentstep_downface_19_20869c6c4df1517848398693617bdfc2 _currentstep_runenumber_21_3c4aa7924813a36f10d2fa820589c6a3) (Utilities|Struct|BreakCubeStruct (Variables|/Figment_S05_Cube/Athena/Prototype/Blueprints/Cube/CUBE_Figment.CUBE_Figment_C|GetCurrentStep_T_16_EA4278F841EE4E73C6861E88CB99517B _output_get)))
    (if (Math|Byte|Equal(Byte) _currentstep_runenumber_21_3c4aa7924813a36f10d2fa820589c6a3 (Math|Byte|Byte+Byte (Variables|Default|GetRuneIndex) 1))
      (Transformation|AttachComponentToComponent (Variables|Default|GetP_Cube_Printing) (Variables|/Figment_S05_Cube/Athena/Prototype/Blueprints/Cube/CUBE_Figment.CUBE_Figment_C|GetStaticMesh (Variables|Default|GetCubeFigment)) (Utilities|String|StringToName (Utilities|String|ToString(Byte) _currentstep_downface_19_20869c6c4df1517848398693617bdfc2)) "SnapToTarget" "SnapToTarget" "SnapToTarget" true)
      (Transformation|SetRelativeRotation _p_cube_printing "180.000000, 0, 0" false false)
      (Components|Activation|SetActive _p_cube_printing true true)
      (Default|BindEventtoStopRuneGlow (Variables|Default|GetCubeFigment) (EventDispatchers|CreateEvent :!event DisableGlowEffect)))
    (else
      (Components|Activation|SetActive (Variables|Default|GetP_Cube_Printing) false false))))

(fn OnRep_RuneIndex
  (bind _ascube_figment (Utilities|Casting|CastToCUBE_Figment (Actor|GetOwner)))
  (Utilities|FlowControl|Sequence
    (:then_1
      (bind (_output_location _output_rotation) (Math|Transform|BreakTransform (Utilities|Array|Get(acopy) (Variables|/Figment_S05_Cube/Athena/Prototype/Blueprints/Cube/CUBE_Figment.CUBE_Figment_C|GetRuneLocations _ascube_figment) (Math|Conversions|ToInteger(Byte) (Variables|Default|GetRuneIndex)))))
      (Transformation|SetWorldTransform (Variables|Default|GetRuneMesh) (Math|Transform|MakeTransform _output_location _output_rotation "0.520000,0.520000,0.520000") false false)
      (CallFunction|SetRuneMaterial))))

(fn SetRuneMaterial
  (Rendering|Material|SetMaterial (Variables|Default|GetRuneMesh) 0 (Utilities|Select :Index (Math|Conversions|ToInteger(Byte) (Variables|Default|GetRuneIndex)) :Option%200 "/Figment_S05_Cube/Athena/Prototype/Blueprints/Cube/Materials/MI_Figment_Cube_Runes_A.MI_Figment_Cube_Runes_A" :Option%201 "/Figment_S05_Cube/Athena/Prototype/Blueprints/Cube/Materials/MI_Figment_Cube_Runes_B.MI_Figment_Cube_Runes_B" :Option%202 "/Figment_S05_Cube/Athena/Prototype/Blueprints/Cube/Materials/MI_Figment_Cube_Runes_C.MI_Figment_Cube_Runes_C" :Option%203 "/Figment_S05_Cube/Athena/Prototype/Blueprints/Cube/Materials/MI_Figment_Cube_Runes_D.MI_Figment_Cube_Runes_D" :Option%204 "/Figment_S05_Cube/Athena/Prototype/Blueprints/Cube/Materials/MI_Figment_Cube_Runes_E.MI_Figment_Cube_Runes_E" :Option%205 "/Figment_S05_Cube/Athena/Prototype/Blueprints/Cube/Materials/MI_Figment_Cube_Runes_F.MI_Figment_Cube_Runes_F" :Option%206 "/Figment_S05_Cube/Athena/Prototype/Blueprints/Cube/Materials/MI_Figment_Cube_Runes_G.MI_Figment_Cube_Runes_G" :Option%207 "/Figment_S05_Cube/Athena/Prototype/Blueprints/Cube/Materials/MI_Figment_Cube_Runes_H.MI_Figment_Cube_Runes_H")))

(fn ConstructionScript
  (bind _effect_area (Variables|Default|GetEffect%20Area))
  (bind _capsuleradius (Variables|/Script/Engine.CapsuleComponent|GetCapsuleRadius _effect_area))
  (bind _returnvalue (/ _capsuleradius 1e+03))
  (bind _areasphere (Variables|Default|GetAreaSphere))
  (Transformation|SetWorldScale3D _areasphere (Math|Conversions|ToVector(Float) _returnvalue))
  (Rendering|Material|SetScalarParameterValueonMaterials _areasphere "SphereRadius" _capsuleradius)
  (Effects|Components|ParticleSystem|SetVectorParameter (Variables|Default|GetP_Cube_AoE_Volume) "SizeScale" (Math|Conversions|ToVector(Float) _returnvalue)))
