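; Cube data-layer manager that swaps Tilted Towers building destruction states as CUBE_Figment advances along its path.
; A looping five-second timer runs both stage checks and sets IsNewGame from server time.
; Stage 1 compares PathIndex to its threshold, keeping DL0 active below it or delaying, then swapping to DL1 once.
; Stage 2 activates DL2 and unloads the others once; its index check gates only the delay, not the swap.
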(event Custom|Check%20New%20Game
  ; Check if the game just started
  (if (> (Fortnite|Sync|Time|GetServerWorldRealTimeSeconds) 3e+01)
    (Variables|Default|SetIsNewGame false)
    (else
      (Variables|Default|SetIsNewGame true))))

(event Custom|DisplayPathIndex)

(event EventBeginPlay
  ; Begin Play
  (Utilities|FlowControl|Sequence
    (:then_0
      (Utilities|Time|SetTimerbyEvent (delegate Custom|DisplayPathIndex) 5.0 true))
    (:then_1
      (Utilities|Time|SetTimerbyEvent (delegate Custom|Check%20Stage) 5.0 true))))

(event Custom|Check%20Stage
  ; Detect actors in box and destroy if they have specified tag
  (Utilities|FlowControl|Sequence
    (:then_0
      (CallFunction|ExecuteStage1))
    (:then_1
      (CallFunction|ExecuteStage2))
    (:then_2
      (CallFunction|CheckNewGame))))

(event Custom|Execute%20Stage%201
  (bind _dl0_name (Variables|Default|GetDL0%20Name))
  (bind _dl0_name_2 (Variables|Default|GetDL0%20Name))
  (bind _dl1_name (Variables|Default|GetDL1%20Name))
  ; Activate Half Destroyed Building in Tilted Towers
  (bind _outactors (Actor|GetAllActorsOfClass "/Figment_S05_Cube/Athena/Prototype/Blueprints/Cube/CUBE_Figment.CUBE_Figment_C"))
  (for _array_element _outactors
    (Utilities|IsValid _array_element
      (:Is%20Valid
        (if (>= (Variables|/Figment_S05_Cube/Athena/Prototype/Blueprints/Cube/CUBE_Figment.CUBE_Figment_C|GetPathIndex _array_element) (Variables|Default|GetStage%201%20Index))
          (Utilities|FlowControl|Delay 1e+01)
          (Utilities|FlowControl|DoOnce false
            (:Completed
              (bind _returnvalue (WorldPartition|GetDataLayerManager))
              (Utilities|IsValid (DataLayers|GetDataLayerInstancefromName _returnvalue _dl0_name)
                (:Is%20Valid
                  (DataLayers|SetDataLayerInstanceRuntimeState _returnvalue (DataLayers|GetDataLayerInstancefromName _returnvalue _dl0_name) "Unloaded" false)
                  (Utilities|IsValid (DataLayers|GetDataLayerInstancefromName _returnvalue _dl1_name)
                    (:Is%20Valid
                      (DataLayers|SetDataLayerInstanceRuntimeState _returnvalue (DataLayers|GetDataLayerInstancefromName _returnvalue _dl1_name) "Activated" false))
                    (:Is%20Not%20Valid
                      (Development|PrintString "No Valid DL Found!" true true "(R=1.000000,G=0.000000,B=0.043093,A=1.000000)" 25.0))))
                (:Is%20Not%20Valid
                  (Development|PrintString "No Valid DL Found!" true true "(R=1.000000,G=0.000000,B=0.043093,A=1.000000)" 25.0)))))
          (else
            (bind _returnvalue_2 (WorldPartition|GetDataLayerManager))
            (Utilities|IsValid (DataLayers|GetDataLayerInstancefromName _returnvalue_2 _dl0_name_2)
              (:Is%20Valid
                (DataLayers|SetDataLayerInstanceRuntimeState _returnvalue_2 (DataLayers|GetDataLayerInstancefromName _returnvalue_2 _dl0_name_2) "Activated" false)))))))))

(event Custom|Execute%20Stage%202
  ; Activate Full Destroyed Building in Tilted Towers
  (bind _outactors (Actor|GetAllActorsOfClass "/Figment_S05_Cube/Athena/Prototype/Blueprints/Cube/CUBE_Figment.CUBE_Figment_C"))
  (for _array_element _outactors
    (Utilities|IsValid _array_element
      (:Is%20Valid
        (if (>= (Variables|/Figment_S05_Cube/Athena/Prototype/Blueprints/Cube/CUBE_Figment.CUBE_Figment_C|GetPathIndex _array_element) (Variables|Default|GetStage%202%20Index))
          (Utilities|FlowControl|Delay 1e+01))
        (Utilities|FlowControl|DoOnce false
          (:Completed
            (bind _returnvalue (WorldPartition|GetDataLayerManager))
            (DataLayers|SetDataLayerInstanceRuntimeState _returnvalue (DataLayers|GetDataLayerInstancefromName _returnvalue (Variables|Default|GetDL0%20Name)) "Unloaded" false)
            (DataLayers|SetDataLayerInstanceRuntimeState _returnvalue (DataLayers|GetDataLayerInstancefromName _returnvalue (Variables|Default|GetDL1%20Name)) "Unloaded" false)
            (DataLayers|SetDataLayerInstanceRuntimeState _returnvalue (DataLayers|GetDataLayerInstancefromName _returnvalue (Variables|Default|GetDL2%20Name)) "Activated" false)))))))

(fn ConstructionScript)
