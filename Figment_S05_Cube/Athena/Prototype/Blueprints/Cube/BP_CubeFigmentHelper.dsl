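; Helper actor for the Figment cube that clears placed actors and loot along its path, for every path point at or below the cube's current index.
; BeginPlay finds the cube and cleans up immediately if its path index is set, otherwise binds to setup-complete and, optionally on authority, rune-landing events.
; CleanUpPath silently kills building actors and box-traces overlap volumes to destroy pickups, while ClientCleanup only hides actors.
; The construction script draws those volumes as debug boxes.
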(event Custom|ClientCleanup
  (params (CurrentIndex :type Integer))
  (for _array_element (Variables|Default|GetPathPoints)
    (bind (_array_element_actors_6_675d370e42be898aed26fbb8a3441e47 _array_element_pathpoint_7_c97c5dc24223fca25dca3f86cc13d01c) (Utilities|Struct|BreakPathPoints _array_element))
    (if (>= CurrentIndex _array_element_pathpoint_7_c97c5dc24223fca25dca3f86cc13d01c)
      (for _array_element_2 _array_element_actors_6_675d370e42be898aed26fbb8a3441e47
        (bind _asbuilding_actor (Utilities|Casting|CastToBuildingActor _array_element_2))
        (Rendering|SetActorHiddenInGame _asbuilding_actor true)))))

(event EventBeginPlay
  (bind _cubefigment (Variables|Default|GetCubeFigment))
  (bind _outactors (Actor|GetAllActorsOfClass "/Figment_S05_Cube/Athena/Prototype/Blueprints/Cube/CUBE_Figment.CUBE_Figment_C"))
  (bind _output_get (Variables|Default|SetCubeFigment (Utilities|Array|Get(acopy) _outactors)))
  (if (== (Variables|/Figment_S05_Cube/Athena/Prototype/Blueprints/Cube/CUBE_Figment.CUBE_Figment_C|GetPathIndex _output_get) -1)
    (Networking|SwitchHasAuthority
      (:Authority
        (Default|BindEventtoCubeSetupComplete _cubefigment (delegate Custom|CleanUpPath))
        (if (Variables|Default|GetRuneCleanup)
          (Default|BindEventtoLandedatRune (Variables|Default|GetCubeFigment) (EventDispatchers|CreateEvent :!event Custom|CleanUpPath))))
      (:Remote
        (Default|BindEventtoCubeSetupComplete _cubefigment (delegate Custom|ClientCleanup))))
    (else
      (CallFunction|CleanUpPath (Variables|/Figment_S05_Cube/Athena/Prototype/Blueprints/Cube/CUBE_Figment.CUBE_Figment_C|GetPathIndex _output_get)))))

(event Custom|CleanUpPath
  (params (CurrentIndex :type Integer))
  (for _array_element (Variables|Default|GetPathPoints)
    (bind (_array_element_actors_6_675d370e42be898aed26fbb8a3441e47 _array_element_pathpoint_7_c97c5dc24223fca25dca3f86cc13d01c _array_element_overlapvolume_10_a0320a4b4306ac684bb2ada37666a9a5) (Utilities|Struct|BreakPathPoints _array_element))
    (if (>= CurrentIndex _array_element_pathpoint_7_c97c5dc24223fca25dca3f86cc13d01c)
      (for _array_element_2 _array_element_actors_6_675d370e42be898aed26fbb8a3441e47
        (bind _asbuilding_actor (Utilities|Casting|CastToBuildingActor _array_element_2))
        (Gameplay|SilentDie _asbuilding_actor false :!class /Script/FortniteGame.BuildingActor))
      (if (not (Math|Transform|Equal(Transform) _array_element_overlapvolume_10_a0320a4b4306ac684bb2ada37666a9a5 (struct :Location "0, 0, 0" :Rotation "0, 0, 0" :Scale "1.000000,1.000000,1.000000")))
        (bind (_outhits _returnvalue) (Collision|MultiBoxTraceForObjects :Start (.location (Math|Transform|ComposeTransforms _array_element_overlapvolume_10_a0320a4b4306ac684bb2ada37666a9a5 (Transformation|GetActorTransform))) :End (.location (Math|Transform|ComposeTransforms _array_element_overlapvolume_10_a0320a4b4306ac684bb2ada37666a9a5 (Transformation|GetActorTransform))) :HalfSize (* (.scale (Math|Transform|ComposeTransforms _array_element_overlapvolume_10_a0320a4b4306ac684bb2ada37666a9a5 (Transformation|GetActorTransform))) 768.0) :Orientation "0, 0, 0" :ObjectTypes (Utilities|Array|MakeArray "ObjectTypeQuery4") :bTraceComplex false :DrawDebugType "Persistent" :bIgnoreSelf true :TraceColor "(R=1.000000,G=0.000000,B=0.000000,A=1.000000)" :TraceHitColor "(R=0.000000,G=1.000000,B=0.000000,A=1.000000)" :DrawTime 5.0))
        (if _returnvalue
          (for _array_element_3 _outhits
            (bind _asfort_pickup (Utilities|Casting|CastToFortPickup (.HitActor (Collision|BreakHitResult _array_element_3))))
            (Actor|DestroyActor _asfort_pickup)))))))

(fn ConstructionScript
  (for _array_element (Variables|Default|GetPathPoints)
    (bind (_array_element_actors_6_675d370e42be898aed26fbb8a3441e47 _array_element_pathpoint_7_c97c5dc24223fca25dca3f86cc13d01c _array_element_overlapvolume_10_a0320a4b4306ac684bb2ada37666a9a5) (Utilities|Struct|BreakPathPoints _array_element))
    (if (not (Math|Transform|Equal(Transform) _array_element_overlapvolume_10_a0320a4b4306ac684bb2ada37666a9a5 (struct :Location "0, 0, 0" :Rotation "0, 0, 0" :Scale "1.000000,1.000000,1.000000")))
      (Rendering|Debug|DrawDebugBox :Center (.location (Math|Transform|ComposeTransforms _array_element_overlapvolume_10_a0320a4b4306ac684bb2ada37666a9a5 (Transformation|GetActorTransform))) :Extent (* (.scale (Math|Transform|ComposeTransforms _array_element_overlapvolume_10_a0320a4b4306ac684bb2ada37666a9a5 (Transformation|GetActorTransform))) 768.0) :LineColor "(R=1.000000,G=0.000000,B=0.000000,A=1.000000)" :Duration 0.0 :Thickness 4.0 :DepthPriority "World"))))
