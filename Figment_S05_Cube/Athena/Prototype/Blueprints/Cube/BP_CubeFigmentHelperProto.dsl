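; Prototype helper actor for the Figment cube that clears tagged actors from its overlap box once a cube reaches a set path index.
; BeginPlay starts two looping timers that poll every CUBE_Figment actor through GetAllActorsOfClass.
; The slower timer prints each cube's path index as debug text.
; The faster one compares that index to Helper Path Index, then box-overlaps OverlapCollision and destroys actors carrying Actor Tag.
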(event Custom|DisplayPathIndex
  ; Display current Path Index
  (bind _outactors (Actor|GetAllActorsOfClass "/Figment_S05_Cube/Athena/Prototype/Blueprints/Cube/CUBE_Figment.CUBE_Figment_C"))
  (for _array_element _outactors
    (Utilities|IsValid _array_element
      (:Is%20Valid
        (Development|PrintString (Utilities|String|Append "Cube Path Index Is: " (Utilities|String|ToString(Integer) (Variables|/Figment_S05_Cube/Athena/Prototype/Blueprints/Cube/CUBE_Figment.CUBE_Figment_C|GetPathIndex _array_element))) true true "(R=1.000000,G=0.000000,B=0.856251,A=1.000000)" 5.0)))))

(event EventBeginPlay
  ; Begin Play
  (Utilities|FlowControl|Sequence
    (:then_0
      (Utilities|Time|SetTimerbyEvent (delegate Custom|DisplayPathIndex) 45.0 true))
    (:then_1
      (Utilities|Time|SetTimerbyEvent (delegate Custom|Destroy%20Overlap%20Actors) 5.0 true))))

(event Custom|Destroy%20Overlap%20Actors
  ; Detect actors in box and destroy if they have specified tag
  (bind _outactors (Actor|GetAllActorsOfClass "/Figment_S05_Cube/Athena/Prototype/Blueprints/Cube/CUBE_Figment.CUBE_Figment_C"))
  (for _array_element _outactors
    (Utilities|IsValid _array_element
      (:Is%20Valid
        (if (>= (Variables|/Figment_S05_Cube/Athena/Prototype/Blueprints/Cube/CUBE_Figment.CUBE_Figment_C|GetPathIndex _array_element) (Variables|Default|GetHelper%20Path%20Index))
          (bind _outactors_2 (Collision|BoxOverlapActors (Transformation|GetWorldLocation (Variables|Default|GetOverlapCollision)) (Components|Box|GetScaledBoxExtent (Variables|Default|GetOverlapCollision)) (Utilities|Array|MakeArray "ObjectTypeQuery1" "ObjectTypeQuery6") :!class /Script/Engine.KismetSystemLibrary))
          (for _array_element_2 _outactors_2
            (Utilities|IsValid _array_element_2
              (:Is%20Valid
                (for _array_element_3 (Variables|/Script/Engine.Actor|GetTags _array_element_2)
                  (if (== _array_element_3 (Variables|Default|GetActor%20Tag))
                    (Actor|DestroyActor _array_element_2)))))))))))

(fn ConstructionScript)
