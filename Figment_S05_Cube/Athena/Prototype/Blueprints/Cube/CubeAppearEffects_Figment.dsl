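; Cube appearance effects Blueprint for Figment Season 5 that plays a lightning strike and knocks players away from the hit location.
; LightingStrikeEffects enables server overlaps and repeats LaunchPlayers on a roughly one-second timer until the beam duration ends.
; LaunchPlayers detaches overlapping pawns from remote-controlled pawns, then launches them outward with distance falloff and minimum lift.
; Clients play camera shakes, repeated for long beams, plus timed impact particles and looping force feedback.
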(event EventBeginPlay)

(event Custom|LightingStrikeEffects
  (params (HitLocation :type Vector) (Duration :type Float%20%28double-precision%29))
  (bind _loopingforcefeedback (Variables|Default|GetLoopingForceFeedback))
  (bind _hitlocation (Variables|Default|GetHitLocation))
  (Variables|Default|SetHitLocation HitLocation)
  (Variables|Default|SetBeamDuration Duration)
  (Utilities|FlowControl|Sequence
    (:then_0
      (Networking|SwitchHasAuthority
        (:Authority
          (Collision|SetCollisionEnabled (Variables|Default|GetOverlaps) "QueryOnly"))))
    (:then_1
      (bind _returnvalue (Utilities|Time|SetTimerbyEvent (EventDispatchers|CreateEvent :!event Custom|LaunchPlayers) 0.98 true))
      (Variables|Default|SetLoopingImpulseTimer _returnvalue)
      (CallFunction|LaunchPlayers))
    (:then_2
      (Networking|SwitchHasAuthority
        (:Remote
          (Camera|PlayWorldCameraShake "/Game/Athena/Prototype/Blueprints/Cube/WarmupEventCameraShake.WarmupEventCameraShake_C" _hitlocation 1e+03 6e+04 3.0) ; Initial camera shake
          (if (> (Variables|Default|GetBeamDuration) 0.0)
            (Utilities|FlowControl|Sequence
              (:then_0
                (bind _beamimpactfx (Variables|Default|GetBeamImpactFX)
                  (:then
                    (Components|DestroyComponent _beamimpactfx))
                  (:else
                    (bind _returnvalue_2 (Effects|SpawnEmitteratLocation "/Figment_S05/Athena/Prototype/Blueprints/Cube/FX/P_Lightning_Strike_Beam_Impact_Red.P_Lightning_Strike_Beam_Impact_Red" _hitlocation "90.000000, 0, 0")) ; Looping beam impact FX
                    (Variables|Default|SetBeamImpactFX _returnvalue_2)
                    (Components|Activation|SetActive _loopingforcefeedback true false)
                    (Transformation|SetWorldLocation _loopingforcefeedback _hitlocation false false)
                    (bind _returnvalue_3 (Utilities|Time|SetTimerbyEvent (delegate Custom|DeactivateEffect) (Variables|Default|GetBeamDuration)))
                    (Variables|Default|SetBeamDurationTimer _returnvalue_3))))
              (:then_1
                (if (> (Variables|Default|GetBeamDuration) 5.0)
                  (CallFunction|LoopingCameraShake)
                  (else
                    (Camera|PlayWorldCameraShake "/Figment_S05/Athena/Prototype/Blueprints/Cube/WarmupEventCameraShake_3Sec.WarmupEventCameraShake_3Sec_C" _hitlocation 5e+02 4e+04 4.0) ; Extended shake for beam
                  ))))))))
    (:then_3
      (Utilities|FlowControl|Delay (Variables|Default|GetBeamDuration))
      (Utilities|Time|ClearandInvalidateTimerbyHandle (Variables|Default|GetLoopingImpulseTimer)))))

(event Custom|LaunchPlayers
  (for _array_element (Collision|GetOverlappingActors (Variables|Default|GetOverlaps) "/Game/Athena/PlayerPawn_Athena.PlayerPawn_Athena_C" :!class /Script/Engine.PrimitiveComponent)
    (bind _asfort_player_pawn (Utilities|Casting|CastToFortPlayerPawn _array_element))
    (Variables|Default|Settarget _asfort_player_pawn)
    (RemoteControlPawn|ForceDetachfromRemoteControlledPawn (Variables|Default|Gettarget))
    (Components|Movement|StopMovementImmediately (Variables|/Script/Engine.Character|GetCharacterMovement (Variables|Default|GetTarget)))
    (Networking|SwitchHasAuthority
      (:Authority
        (Character|LaunchCharacter (Variables|Default|GetTarget) (Math|Vector|MakeVector (.x (* (* (Math|Vector|Normalize (- (Transformation|GetActorLocation (Variables|Default|GetTarget)) (Variables|Default|GetHitLocation))) 5e+03) (Math|Float|MapRangeClamped (Math|Vector|VectorLength (- (Transformation|GetActorLocation (Variables|Default|GetTarget)) (Variables|Default|GetHitLocation))) 0.0 1.2e+03 1.0 0.25))) (.y (* (* (Math|Vector|Normalize (- (Transformation|GetActorLocation (Variables|Default|GetTarget)) (Variables|Default|GetHitLocation))) 5e+03) (Math|Float|MapRangeClamped (Math|Vector|VectorLength (- (Transformation|GetActorLocation (Variables|Default|GetTarget)) (Variables|Default|GetHitLocation))) 0.0 1.2e+03 1.0 0.25))) (Math|Float|Max(Float) (+ (.z (* (* (Math|Vector|Normalize (- (Transformation|GetActorLocation (Variables|Default|GetTarget)) (Variables|Default|GetHitLocation))) 5e+03) (Math|Float|MapRangeClamped (Math|Vector|VectorLength (- (Transformation|GetActorLocation (Variables|Default|GetTarget)) (Variables|Default|GetHitLocation))) 0.0 1.2e+03 1.0 0.25))) 2e+02) 6e+02)) true true)))))

(event Custom|CubeAppearEffects)

(event Custom|DeactivateEffect
  (Components|Activation|SetActive (Variables|Default|GetBeamImpactFX) false false)
  (Utilities|Time|ClearandInvalidateTimerbyHandle (Variables|Default|GetBeamDurationTimer)))

(event Custom|LoopingCameraShake
  (if (not (< (Utilities|Time|GetTimerRemainingTimebyHandle (Variables|Default|GetBeamDurationTimer)) 2.0))
    (Camera|PlayWorldCameraShake "/Figment_S05/Athena/Prototype/Blueprints/Cube/WarmupEventCameraShake_3Sec.WarmupEventCameraShake_3Sec_C" (Variables|Default|GetHitLocation) 5e+02 4e+04 4.0) ; Extended shake for beam
    (Utilities|Time|SetTimerbyEvent (delegate Custom|LoopingCameraShake) 2.0)))

(event Collision|EventActorBeginOverlap
  (params (OtherActor :type Actor%20Object%20Reference)))

(event EventTick
  (params (DeltaSeconds :type Float%20%28single-precision%29)))

(fn ConstructionScript)
